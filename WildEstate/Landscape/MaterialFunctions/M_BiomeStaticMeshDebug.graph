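# M_BiomeStaticMeshDebug is a Default Lit material that, by its name, serves to debug WildEstate biome masks on static meshes.
# It holds a single call to MF_WildEstate_BiomeMasks, with a static bool at its default false on UseLandscapeCoords.
# The translucent projection input is left unconnected, and no link from the function to a material output appears.
# Automatic usage flagging in the editor is off.
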Material /WildEstate/Landscape/MaterialFunctions/M_BiomeStaticMeshDebug
graph {
 properties {bAutomaticallySetUsageInEditor=false, usedShadingModels=Default Lit}
 nodes [MaterialFunctionCall!(UseLandscapeCoords=StaticBool(), "UseProjection(translucent)") {materialFunction=&/MF_WildEstate_BiomeMasks}]
}
